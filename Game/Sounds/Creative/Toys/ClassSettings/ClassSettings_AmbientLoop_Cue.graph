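# Sound cue for the looping ambient sound of the ClassSettings toy in Creative, routed to the Creative gadgets sound class at low volume.
# A single looping wave player feeds the output directly.
# Inline attenuation uses a short falloff with stereo spread and listener focus, quieting the sound when out of focus.
# Inline concurrency caps simultaneous instances, stopping the quietest when the limit is exceeded.
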SoundCue /Game/Sounds/Creative/Toys/ClassSettings/ClassSettings_AmbientLoop_Cue
root=WavePlayer_1
nodes [{id=WavePlayer_1, properties {bLooping=true, soundWaveAssetPtr=&/ClassSettings_AmbientLoop_01}}]
properties {
 bOverrideAttenuation=true
 bOverrideConcurrency=true
 maxDistance=1200
 soundClassObject=&Fort_Gadgets_Creative
 volumeMultiplier=0.15
 attenuationOverrides {
  bEnableFocusInterpolation=true
  bEnableListenerFocus=true
  falloffDistance=1200
  nonFocusVolumeAttenuation=0.5
  stereoSpread=500
  attenuationShapeExtents {x=0}
 }
 concurrencyOverrides {maxCount=3, resolutionRule=StopQuietest}
}
